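; Gameplay ability from the Battle Pass Season 20 hydraulic parts quest that spawns quest actors and attaches them to a scene component.
; It overrides SpawnQuestActorsFromDropEntry, calling the parent implementation before looping over an actor array.
; When the gameplay event's OptionalObject is valid, it is cast to a scene component and each actor is attached there, snapping location to target while keeping world rotation and scale.
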(fn SpawnQuestActorsFromDropEntry
  (params (DropEntry :type Struct%20Spawned%20Quest%20Actor%20Drop%20Entry%20Structure) (EventData :type Gameplay%20Event%20Data%20Structure))
  (bind (_eventtag _instigator _target _optionalobject) (Utilities|Struct|BreakGameplayEventData EventData))
  (|CallParentFunction DropEntry EventData)
  (for _array_element 0
    (Utilities|IsValid _optionalobject
      (:Is%20Valid
        (bind _asscene_component (Utilities|Casting|CastToSceneComponent _optionalobject))
        (Transformation|AttachActorToComponent _array_element _asscene_component "None" "SnapToTarget" "KeepWorld" "KeepWorld" true)))))
